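# Niagara system for the Magical vehicle boost cosmetic, drawing layered trails and flare shapes at the car's filtered boost sockets, offset by User.DTCarOffset.
# A ribbon emitter drives four ribbon renderers, each with its own color and width.
# A merged emitter uses spawn groups as render tags for four mesh renderers and one sprite renderer, with group 3 spawning continuously.
# User.bIsPainted switches default magical gradients to painted-variant colors and brightness.
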NiagaraSystem /VehicleCosmetics/Boosts/Boost_Magical/FX/NS_Boost_Magical
dependencies {
 usedDataInterfaces [
  &NiagaraDataInterfaceColorCurve
  &NiagaraDataInterfaceCurve
  &NiagaraDataInterfaceVectorCurve
  &NiagaraDataInterfaceSkeletalMesh
 ]
 usedDynamicInputs []
 usedModules [
  &self:SystemSpawnScript.SystemScriptSource.SystemScriptGraph.NiagaraNodeAssignment_0.SetVariables_#1
  &SystemState
 ]
}
emitters {
 NE_VC_Boost_Ribbon_Parent002 {
  emitterUpdateScript [
   &EmitterState()
   &VC_FilteredSocketCount()
   &VC_RateQualityHandler()
   &VC_SpawnRibbon_v1(Number of Ribbons=$Emitter.PortCount, SpawnRate=$Emitter.Spawn Rate)
  ]
  particleSpawnScript [
   &VC_PaintedVariantsOverride(Crimson=<rgba 0.4635 0 0.0126 1>, Purple=<rgba 0.147 0.0176 0.6795 1>)
   &VC_InitializeRibbon_v1(
    Color=$Particles.Color
    Custom Offset=true
    Enable Skeletal Position=true
    Lifetime=0.8
    Local Vector Offset=&AddVectorToPosition(Position=<-40 0 0>, Vector=$User.DTCarOffset)
    Offset Amount=1
    Ribbon Width=40
   )
   &AddVelocity(Velocity=<-1500 0 0>, Velocity Speed Scale=2) {module=AddVelocity001}
  ]
  particleUpdateScript [
   &ParticleState()
   &VC_PaintedVariantsOverride(
    BurntSienna=<rgba 0.0642 0.0198 0.0079 1>
    Cobalt=<rgba 0 0.0185 1 1>
    Crimson=<rgba 0.4635 0 0.0126 1>
    ForestGreen=<rgba 0 0.5 0.0253 1>
    Grey=<rgba 0.0781 0.0781 0.0781 1>
    Lime=<rgba 0.2943 0.5885 0 1>
    Orange=<rgba 0.6724 0.2346 0 1>
    Purple=<rgba 0.0844 0.0482 0.3372 1>
    R1 Black:LinearColor=&ColorFromCurve(ColorCurve=DataInterface() {propertyValues {
       AlphaCurve {keys [<at 0 1 RCIM_Cubic>,<at 1 0 RCIM_Cubic>]}
       BlueCurve {keys [<at 0 0.0521 RCIM_Cubic>,<at 0.1166 0.01 RCIM_Cubic>]}
       GreenCurve {keys [<at 0 0.0521 RCIM_Cubic>,<at 0.1166 0.01 RCIM_Cubic>]}
       RedCurve {keys [<at 0 0.0521 RCIM_Cubic>,<at 0.1166 0.01 RCIM_Cubic>]}
      }})
    R1 BurntSienna:LinearColor=&ColorFromCurve(ColorCurve=DataInterface() {propertyValues {
       AlphaCurve {keys [<at 0 1 RCIM_Cubic>,<at 1 0 RCIM_Cubic>]}
       BlueCurve {keys [<at 0 0.0482 RCIM_Cubic>,<at 0.2929 0.0107 RCIM_Cubic>]}
       GreenCurve {keys [<at 0 0.0887 RCIM_Cubic>,<at 0.2929 0.0196 RCIM_Cubic>]}
       RedCurve {keys [<at 0 0.4233 RCIM_Cubic>,<at 0.2929 0.0938 RCIM_Cubic>]}
      }})
    R1 Cobalt:LinearColor=&ColorFromCurve(ColorCurve=DataInterface() {propertyValues {
       AlphaCurve {keys [<at 0 1 RCIM_Cubic>,<at 1 0 RCIM_Cubic>]}
       BlueCurve {keys [<at 0 0.3594 RCIM_Cubic>,<at 0.2929 0.7231 RCIM_Cubic>]}
       GreenCurve {keys [<at 0 0.115 RCIM_Cubic>,<at 0.2929 0.1022 RCIM_Cubic>]}
       RedCurve {keys [<at 0 0.1104 RCIM_Cubic>,<at 0.2929 0.0056 RCIM_Cubic>]}
      }})
    R1 Crimson:LinearColor=&ColorFromCurve(ColorCurve=DataInterface() {propertyValues {
       AlphaCurve {keys [<at 0 1 RCIM_Cubic>,<at 1 0 RCIM_Cubic>]}
       BlueCurve {keys [<at 0 0.0755 RCIM_Cubic>,<at 0.2929 0.0123 RCIM_Cubic>]}
       GreenCurve {keys [<at 0 0.065 RCIM_Cubic>,<at 0.2929 0 RCIM_Cubic>]}
       RedCurve {keys [<at 0 0.4621 RCIM_Cubic>,<at 0.2929 0.4621 RCIM_Cubic>]}
      }})
    R1 ForestGreen:LinearColor=&ColorFromCurve(ColorCurve=DataInterface() {propertyValues {
       AlphaCurve {keys [<at 0 1 RCIM_Cubic>,<at 1 0 RCIM_Cubic>]}
       BlueCurve {keys [<at 0 0.1981 RCIM_Cubic>,<at 0.2929 0.028 RCIM_Cubic>]}
       GreenCurve {keys [<at 0 0.5781 RCIM_Cubic>,<at 0.2929 0.5029 RCIM_Cubic>]}
       RedCurve {keys [<at 0 0.2223 RCIM_Cubic>,<at 0.2929 0.1087 RCIM_Cubic>]}
      }})
    R1 Grey:LinearColor=&ColorFromCurve(ColorCurve=DataInterface() {propertyValues {
       AlphaCurve {keys [<at 0 1 RCIM_Cubic>,<at 1 0 RCIM_Cubic>]}
       BlueCurve {keys [<at 0 0.151 RCIM_Cubic>,<at 0.2929 0.0521 RCIM_Cubic>]}
       GreenCurve {keys [<at 0 0.151 RCIM_Cubic>,<at 0.2929 0.0521 RCIM_Cubic>]}
       RedCurve {keys [<at 0 0.151 RCIM_Cubic>,<at 0.2929 0.0521 RCIM_Cubic>]}
      }})
    R1 Lime:LinearColor=&ColorFromCurve(ColorCurve=DataInterface() {propertyValues {
       AlphaCurve {keys [<at 0 1 RCIM_Cubic>,<at 1 0 RCIM_Cubic>]}
       BlueCurve {keys [<at 0 0.0431 RCIM_Cubic>,<at 0.2126 0 RCIM_Cubic>]}
       GreenCurve {keys [<at 0 0.5906 RCIM_Cubic>,<at 0.2126 0.3958 RCIM_Cubic>]}
       RedCurve {keys [<at 0 0.3176 RCIM_Cubic>,<at 0.2126 0.1985 RCIM_Cubic>]}
      }})
    R1 Orange:LinearColor=&ColorFromCurve(ColorCurve=DataInterface() {propertyValues {
       AlphaCurve {keys [<at 0 1 RCIM_Cubic>,<at 1 0 RCIM_Cubic>]}
       BlueCurve {keys [<at 0 0 RCIM_Cubic>,<at 0.2126 0.0028 RCIM_Cubic>]}
       GreenCurve {keys [<at 0 0.2335 RCIM_Cubic>,<at 0.2126 0.0221 RCIM_Cubic>]}
       RedCurve {keys [<at 0 0.707 RCIM_Cubic>,<at 0.2126 0.2917 RCIM_Cubic>]}
      }})
    R1 Pink:LinearColor=&ColorFromCurve(ColorCurve=DataInterface() {propertyValues {
       AlphaCurve {keys [<at 0 1 RCIM_Cubic>,<at 1 0 RCIM_Cubic>]}
       BlueCurve {keys [<at 0 0.237 RCIM_Cubic>,<at 0.2126 0.3419 RCIM_Cubic>]}
       GreenCurve {keys [<at 0 0.1507 RCIM_Cubic>,<at 0.2126 0.0612 RCIM_Cubic>]}
       RedCurve {keys [<at 0 0.2865 RCIM_Cubic>,<at 0.2126 0.5029 RCIM_Cubic>]}
      }})
    R1 Purple:LinearColor=&ColorFromCurve(ColorCurve=DataInterface() {propertyValues {
       AlphaCurve {keys [<at 0 1 RCIM_Cubic>,<at 1 0 RCIM_Cubic>]}
       BlueCurve {keys [<at 0 0.3372 RCIM_Cubic>,<at 0.2126 0.4323 RCIM_Cubic>]}
       GreenCurve {keys [<at 0 0.1106 RCIM_Cubic>,<at 0.2126 0.0316 RCIM_Cubic>]}
       RedCurve {keys [<at 0 0.139 RCIM_Cubic>,<at 0.2126 0.0616 RCIM_Cubic>]}
      }})
    R1 Saffron:LinearColor=&ColorFromCurve(ColorCurve=DataInterface() {propertyValues {
       AlphaCurve {keys [<at 0 1 RCIM_Cubic>,<at 1 0 RCIM_Cubic>]}
       BlueCurve {keys [<at 0 0.1051 RCIM_Cubic>,<at 0.2126 0 RCIM_Cubic>]}
       GreenCurve {keys [<at 0 0.252 RCIM_Cubic>,<at 0.2126 0.5227 RCIM_Cubic>]}
       RedCurve {keys [<at 0 0.2969 RCIM_Cubic>,<at 0.2126 0.6823 RCIM_Cubic>]}
      }})
    R1 SkyBlue:LinearColor=&ColorFromCurve(ColorCurve=DataInterface() {propertyValues {
       AlphaCurve {keys [<at 0 1 RCIM_Cubic>,<at 1 0 RCIM_Cubic>]}
       BlueCurve {keys [<at 0 0.4323 RCIM_Cubic>,<at 0.2126 0.3698 RCIM_Cubic>]}
       GreenCurve {keys [<at 0 0.3419 RCIM_Cubic>,<at 0.2126 0.2471 RCIM_Cubic>]}
       RedCurve {keys [<at 0 0.1599 RCIM_Cubic>,<at 0.2126 0 RCIM_Cubic>]}
      }})
    R1 TitaniumWhite:LinearColor=&ColorFromCurve(ColorCurve=DataInterface() {propertyValues {
       AlphaCurve {keys [<at 0 0.9 RCIM_Cubic>,<at 1 0 RCIM_Cubic>]}
       BlueCurve {keys [<at 0 0.3712 RCIM_Cubic>,<at 0.2943 0.8125 RCIM_Cubic>]}
       GreenCurve {keys [<at 0 0.3712 RCIM_Cubic>,<at 0.2943 0.8125 RCIM_Cubic>]}
       RedCurve {keys [<at 0 0.3712 RCIM_Cubic>,<at 0.2943 0.8125 RCIM_Cubic>]}
      }})
    R2 Black:LinearColor=<rgba 0.1094 0.1094 0.1094 1>
    R2 BurntSienna:LinearColor=<rgba 0.4219 0.1303 0.052 1>
    R2 Cobalt:LinearColor=<rgba 0.0625 0.0798 1 1>
    R2 Crimson:LinearColor=<rgba 0.3802 0.0356 0.0448 1>
    R2 Grey:LinearColor=<rgba 0.0417 0.0417 0.0417 1>
    R2 Purple:LinearColor=<rgba 0.0844 0.0482 0.3372 1>
    Saffron=<rgba 0.2708 0.2076 0 1>
    SkyBlue=<rgba 0 0.357 0.5365 1>
    TitaniumWhite=<rgba 1.5 1.5 1.5 1>
    Use R1 Painted Overrides=true
    Use R2 Painted Overrides=true
   ) {module=VC_PaintedVariantsOverride001}
   &ScaleRibbonWidth(Ribbon Width Scale=&FloatFromCurve(FloatCurve=DataInterface() {propertyValues {Curve {keys [
         <at 0 0.4 RCIM_Cubic>
         {arriveTangentWeight=0.0731, interpMode=RCIM_Cubic, leaveTangentWeight=0.2924, tangentMode=RCTM_User, time=0.2, value=0.5}
         <at 1 0.2 RCIM_Cubic>
        ]}}})) {enabled=false}
   &InheritSourceMovement(
    Applied Position Scale=&VectorFromFloat(Value=&FloatFromCurve(FloatCurve=DataInterface() {propertyValues {Curve {keys [<at 0 1 RCIM_Cubic>,<at 1 0 RCIM_Cubic>]}}}))
    ApplyForce=false
    ApplyPosition=true
    ApplyPositionFalloff=false
   )
   &SolveForcesAndVelocity(Clamp Velocity=false, Limit Acceleration=false)
   &VC_UpdateRibbon_v1(Enable Skeletal Position=true)
   &DynamicMaterialParameters(
    Index 0 Param 1=2
    Index 0 Param 2=0.5
    Index 0 Param 3=1
    Index 0 Param 4=1
    Module.Index0Float0Write=true
    Module.Index0Float1Write=true
    Module.Index0Float2Write=true
    Module.Index0Float3Write=true
   )
   /SetVariables(
    Particles.Index1:Vector4f=&MakeVector4(X=3, Y=3, Z=1, W=-1)
    Particles.Index2:Vector4f=&MakeVector4(X=2, Y=1, Z=0, W=0)
    Particles.Index3:Vector4f=&MakeVector4(X=2, Y=&MakeCustomFloatFromBool(Bool=$User.bIsPainted, True Float=300, False Float=100), Z=8, W=-1)
    Particles.RibbonColor0:LinearColor=&VC_DynamicScaleColor(
     Color to Scale:LinearColor=$Particles.Color
     RGB Scale:Vector3f=&MakeCustomVectorFromBool(
      Bool=$User.bIsPainted
      True Vector=&VectorFromFloat(Value=&Multiply_Float(A=0.1, B=0.4))
      False Vector=&MakeVectorFromLinearColorRGB(Color=&Multiply_LinearColorByFloat(Linear Color=<rgba 0.08 0.05 0.1 1>, Float=0.4))
     )
     Alpha Scale:NiagaraFloat=&FloatFromCurve(FloatCurve=DataInterface() {propertyValues {Curve {keys [
          <at 0 0.5 RCIM_Cubic>
          {arriveTangentWeight=0.1572, interpMode=RCIM_Cubic, leaveTangentWeight=0.1572, tangentMode=RCTM_User, time=0.3, value=1}
          <at 0.7 0 RCIM_Cubic>
         ]}}})
    )
    Particles.RibbonColor1:LinearColor=&VC_DynamicScaleColor(
     Color to Scale:LinearColor=&MakeCustomLinearColorFromBool(
      Bool=$User.bIsPainted
      True Color=$Particles.R1 Color
      False Color=&ColorFromCurve(ColorCurve=DataInterface() {propertyValues {
         AlphaCurve {keys [<at 0 1 RCIM_Cubic>,<at 1 1 RCIM_Cubic>]}
         BlueCurve {keys [
           <at 0.0459 0.1458 RCIM_Cubic>
           {arriveTangentWeight=0.1346, interpMode=RCIM_Cubic, leaveTangentWeight=0.1565, tangentMode=RCTM_User, time=0.1, value=1}
           <at 0.25 1 RCIM_Cubic>
          ]}
         GreenCurve {keys [
           <at 0.0459 1 RCIM_Cubic>
           {arriveTangent=-4.5941, interpMode=RCIM_Cubic, leaveTangent=-4.5941, time=0.1, value=0.3005}
           <at 0.25 0.0623 RCIM_Cubic>
          ]}
         RedCurve {keys [
           <at 0.0459 1 RCIM_Cubic>
           {arriveTangentWeight=0.1339, interpMode=RCIM_Cubic, leaveTangentWeight=0.1557, tangentMode=RCTM_User, time=0.1, value=1}
           <at 0.25 0.1503 RCIM_Cubic>
          ]}
        }})
     )
     RGB Scale:Vector3f=&MakeCustomVectorFromBool(
      Bool=$User.bIsPainted
      True Vector=&VectorFromFloat(Value=&FloatFromCurve(FloatCurve=DataInterface() {propertyValues {Curve {keys [<at 0 5 RCIM_Cubic>,<at 0.3 0.5 RCIM_Cubic>]}}}))
      False Vector=&VectorFromFloat(Value=&FloatFromCurve(FloatCurve=DataInterface() {propertyValues {Curve {keys [<at 0 2 RCIM_Cubic>,<at 0.3 0.5 RCIM_Cubic>]}}}))
     )
     Alpha Scale:NiagaraFloat=&FloatFromCurve(FloatCurve=DataInterface() {propertyValues {Curve {keys [
          <at 0 0 RCIM_Cubic>
          {arriveTangentWeight=0.2247, interpMode=RCIM_Cubic, leaveTangentWeight=0.4494, tangentMode=RCTM_User, time=0.1, value=2}
          {arriveTangentWeight=0.0667, interpMode=RCIM_Cubic, leaveTangentWeight=0.0667, tangentMode=RCTM_User, time=0.3, value=2}
          {arriveTangentWeight=0.7723, interpMode=RCIM_Cubic, tangentMode=RCTM_User, time=0.5, value=0}
         ]}}})
    )
    Particles.RibbonColor2:LinearColor=&VC_DynamicScaleColor(
     Color to Scale:LinearColor=&MakeCustomLinearColorFromBool(
      Bool=$User.bIsPainted
      True Color=$Particles.R1 Color
      False Color=&ColorFromCurve(ColorCurve=DataInterface() {propertyValues {
         AlphaCurve {keys [<at 0 1 RCIM_Cubic>,<at 1 1 RCIM_Cubic>]}
         BlueCurve {keys [<at 0.1234 0.4 RCIM_Cubic>,<at 0.1866 1>,<at 0.2559 1 RCIM_Cubic>]}
         GreenCurve {keys [<at 0.1234 1 RCIM_Cubic>,<at 0.1866 0.3>,<at 0.2559 0.1 RCIM_Cubic>]}
         RedCurve {keys [<at 0.1234 1 RCIM_Cubic>,<at 0.1866 1>,<at 0.2559 0.2 RCIM_Cubic>]}
        }})
     )
     RGB Scale:Vector3f=&VectorFromFloat(Value=&FloatFromCurve(FloatCurve=DataInterface() {propertyValues {Curve {keys [<at 0.2 1>,<at 0.5 2>]}}}, Scale Curve=1.5))
     Alpha Scale:NiagaraFloat=&FloatFromCurve(
      FloatCurve=DataInterface() {propertyValues {Curve {keys [
          <at 0 0 RCIM_Cubic>
          {interpMode=RCIM_Cubic, leaveTangentWeight=0.0833, tangentMode=RCTM_User, time=0.2, value=1}
          {arriveTangentWeight=0.1008, interpMode=RCIM_Cubic, tangentMode=RCTM_User, time=0.5, value=0}
         ]}}}
      Scale Curve=4
     )
    )
    Particles.RibbonColor3:LinearColor=&VC_DynamicScaleColor(
     Color to Scale:LinearColor=&MakeCustomLinearColorFromBool(
      Bool=$User.bIsPainted
      True Color=$Particles.R1 Color
      False Color=&ColorFromCurve(ColorCurve=DataInterface() {propertyValues {
         AlphaCurve {keys [<at 0 1 RCIM_Cubic>,<at 1 1 RCIM_Cubic>]}
         BlueCurve {keys [<at 0.0459 0.1458 RCIM_Cubic>,<at 0.1295 1>,<at 0.2417 0.8542>]}
         GreenCurve {keys [<at 0.0459 0.9372 RCIM_Cubic>,<at 0.1295 0.3005>,<at 0.2417 0.0623>]}
         RedCurve {keys [<at 0.0459 1 RCIM_Cubic>,<at 0.1295 1>,<at 0.2417 0.1503>]}
        }})
     )
     RGB Scale:Vector3f=&MakeCustomVectorFromBool(
      Bool=$User.bIsPainted
      True Vector=&VectorFromFloat(Value=&FloatFromCurve(FloatCurve=DataInterface() {propertyValues {Curve {keys [<at 0 50>,<at 0.3 0.5>]}}}))
      False Vector=&VectorFromFloat(Value=1)
     )
     Alpha Scale:NiagaraFloat=&FloatFromCurve(
      FloatCurve=DataInterface() {propertyValues {Curve {keys [
          {arriveTangent=11.5571, interpMode=RCIM_Cubic, leaveTangent=11.5571, tangentMode=RCTM_User, time=0, value=0}
          <at 0.25 1 RCIM_Cubic>
          {arriveTangent=-9.2133, interpMode=RCIM_Cubic, leaveTangent=-9.2133, tangentMode=RCTM_User, time=0.5, value=0}
         ]}}}
      Scale Curve=0.5
     )
    )
    Particles.RibbonWidth0:NiagaraFloat=&Multiply_Float(
     A=&FloatFromCurve(FloatCurve=DataInterface() {propertyValues {Curve {keys [<at 0 1>,<at 1 0>]}}}, Scale Curve=3.5)
     B=$Particles.RibbonWidth
    )
    Particles.RibbonWidth1:NiagaraFloat=&Multiply_Float(
     A=&FloatFromCurve(FloatCurve=DataInterface() {propertyValues {Curve {keys [
          {arriveTangent=-1.5, interpMode=RCIM_Cubic, time=0, value=0.4}
          {arriveTangentWeight=0.1221, interpMode=RCIM_Cubic, leaveTangentWeight=0.2848, tangentMode=RCTM_User, time=0.2, value=0.8}
          <at 0.5 0.2 RCIM_Cubic>
         ]}}})
     B=$Particles.RibbonWidth
    )
    Particles.RibbonWidth2:NiagaraFloat=&Multiply_Float(A=1.5, B=$Particles.RibbonWidth)
    Particles.RibbonWidth3:NiagaraFloat=&Multiply_Float(
     A=&FloatFromCurve(FloatCurve=DataInterface() {propertyValues {Curve {keys [<at 0 0.7>,<at 1 0>]}}})
     B=$Particles.RibbonWidth
    )
   ) {isSetParameters=true}
  ]
  rendererClasses [&NiagaraRibbonRendererProperties]
  renderers [
   {rendererClass=&NiagaraRibbonRendererProperties}
   {rendererClass=&NiagaraRibbonRendererProperties, rendererIndex=1}
   {rendererClass=&NiagaraRibbonRendererProperties, rendererIndex=2}
   {rendererClass=&NiagaraRibbonRendererProperties, rendererIndex=3}
  ]
 }
 NE_VC_Boost_Sprite_Parent002 {
  emitterUpdateScript [
   &EmitterState(
    Inactive Response=::NewEnumerator1
    Life Cycle Mode=::NewEnumerator1
    Loop Behavior=::NewEnumerator1
    Loop Duration Mode=::NewEnumerator1
    UseLoopDelay=false
   )
   &SpawnBurst_Instantaneous(Spawn Count=2, Use Spawn Probability=false, UseLoopCountLimit=false)
   &SpawnBurst_Instantaneous(Spawn Count=2, Spawn Group=1, Use Spawn Probability=false, UseLoopCountLimit=false) {module=SpawnBurst_Instantaneous001}
   &SpawnBurst_Instantaneous(Spawn Count=2, Spawn Group=2, Use Spawn Probability=false, UseLoopCountLimit=false) {module=SpawnBurst_Instantaneous002}
   &SpawnBurst_Instantaneous(Spawn Count=2, Spawn Group=4, Use Spawn Probability=false, UseLoopCountLimit=false) {module=SpawnBurst_Instantaneous004}
   &VC_FilteredSocketCount()
   &SpawnRate(SpawnGroup=3, SpawnRate=10, Use Spawn Probability=false)
   &SpawnBurst_Instantaneous(Spawn Count=2, Spawn Time=2.6, Use Spawn Probability=false, UseLoopCountLimit=false) {module=SpawnBurst_Instantaneous003}
  ]
  particleSpawnScript [
   /SetVariables(Particles.RenderTag:NiagaraInt32=&SpawnGroup()) {isSetParameters=true}
   &VC_PaintedVariantsOverride(
    Black=<rgba 0.0191 0.0191 0.0191 3>
    BurntSienna=<rgba 0.1094 0.0232 0.0125 2>
    Cobalt=<rgba 0.0527 0.0622 0.5625 1>
    Crimson=<rgba 0.3385 0.0194 0.028 1>
    ForestGreen=<rgba 0 0.3073 0.0155 1>
    Grey=<rgba 0.0365 0.0365 0.0365 2>
    Lime=<rgba 0.1726 0.25 0.0104 1>
    Orange=<rgba 0.2552 0.089 0 1>
    Pink=<rgba 0.1771 0.0397 0.1271 1>
    Purple=<rgba 0.087 0.0482 0.388 1>
    R1 BurntSienna:LinearColor=<rgba 0.2188 0.0773 0.0627 1>
    R1 Cobalt:LinearColor=<rgba 0.0427 0.0521 0.5469 1>
    R1 Crimson:LinearColor=<rgba 0.4635 0.0266 0.0384 1>
    R1 DefaultColor=<rgba 1 1 0.2 1>
    R1 ForestGreen:LinearColor=<rgba 0.0942 0.3229 0.1057 1>
    R1 Lime:LinearColor=<rgba 0.244 0.2917 0.0395 1>
    R1 Orange:LinearColor=<rgba 0.5417 0.1889 0 1>
    R1 Purple:LinearColor=<rgba 0.103 0.0482 0.3372 1>
    R1 Saffron:LinearColor=<rgba 0.2865 0.2224 0.0119 1>
    R1 SkyBlue:LinearColor=<rgba 0 0.1975 0.2969 1>
    R1 TitaniumWhite:LinearColor=<rgba 0.2448 0.2448 0.2448 1>
    Saffron=<rgba 0.2188 0.1677 0 1>
    SkyBlue=<rgba 0 0.2079 0.3125 1>
    TitaniumWhite=<rgba 0.1302 0.1302 0.1302 1>
    Use R1 Painted Overrides=true
   )
   &InitializeParticle(
    Color=$Particles.Color
    Mesh Scale=&Multiply_VectorByFloat(Vector=<1 1 3>, Float=0.4)
    Mesh Scale Mode=::NewEnumerator4
    Sprite Rotation Angle=90
    Sprite Rotation Mode=::NewEnumerator0
    Sprite Size=&Multiply_Vector2D_ByFloat(Float=5, Vector2D=<20 6>)
    Sprite Size Mode=::NewEnumerator4
    UsePositionOffset=false
    Write Material Random=true
   )
   &PartitionParticles(Number Of Partitions=$Emitter.PortCount)
   &SkeletalMeshLocation(
    Bone / Socket Index=$Particles.PartitionParticles.Partition
    Bone / Socket Sampling Mode=::NewEnumerator5
    MaskSpawnGroup=false
    Mesh Sampling Type=::NewEnumerator6
    Sampled Position Offset=&AddVectorToPosition(
     Position=/VehicleCosmetics/FX/Niagara/Modules/MergeEmitters/VC_VectorBySpawnGroup(
      SpawnGroup or ExecIndex:NiagaraInt32=0
      CustomInt:NiagaraInt32=0
      MaximumSpawnGroups:NiagaraInt32=5
      Vector Group 0:Vector3f=<-40 0 0>
      Vector Group 1:Vector3f=<-40 0 0>
      Vector Group 2:Vector3f=<-40 0 0>
      Vector Group 3:Vector3f=<-45 0 0>
      Vector Group 4:Vector3f=<-85 0 0>
      Vector Group 5:Vector3f=<0 0 0>
     )
     Vector=$User.DTCarOffset
    )
    Skeletal Mesh=$System.SkeletalMesh
   )
   &AddVelocity(Velocity=<-1500 0 0>) {enabled=false}
   &InitialMeshOrientation(
    Mesh Axis To Orient=<0 0 -1>
    Mesh Orientation Mode=::NewEnumerator2
    Orientation Vector=<0 0 0>
    Rotation=&RandomRangeVector(Minimum=<0 -1 0>, Maximum=<0 1 0>)
    Use Rotation Vector=false
   )
   &SpriteFacingAndAlignment(Sprite Alignment=<1 0 0>, Sprite Facing=<0 1 1>, WriteAlignment=true, WriteFacing=false)
   /SetVariables(
    Particles.ColorTag0:LinearColor=<rgba 1 0.6786 1 0.7304>
    Particles.ColorTag1:LinearColor=<rgba 1 1 0.5052 1>
    Particles.ColorTag2:LinearColor=<rgba 0.3962 0.0696 1 0.3913>
   ) {isSetParameters=true}
  ]
  particleUpdateScript [
   &ParticleState(Kill Particles When Lifetime Has Elapsed=false)
   &SolveForcesAndVelocity(Clamp Velocity=false, Limit Acceleration=false)
   &UpdateMeshOrientation(
    Rotation Rate=/VehicleCosmetics/FX/Niagara/Modules/MergeEmitters/VC_FloatBySpawnGroup(
     SpawnGroup or ExecIndex:NiagaraInt32=0
     CustomInt:NiagaraInt32=0
     MaximumSpawnGroups:NiagaraInt32=4
     Float SG 0:NiagaraFloat=0
     Float SG 1:NiagaraFloat=0
     Float SG 2:NiagaraFloat=0
     Float SG 3:NiagaraFloat=1
     Float SG 4:NiagaraFloat=0
     Float SG 5:NiagaraFloat=0
    )
    Rotation Vector=<0 0 0.4>
   )
   /SetVariables(
    Particles.Index0:Vector4f=&MakeVector4(X=1, Y=0.8, Z=0, W=0)
    Particles.Index1:Vector4f=&MakeVector4(X=1, Y=1, Z=0, W=0)
    Particles.Index2:Vector4f=&MakeVector4(X=1, Y=1.3, Z=0, W=0)
    Particles.MeshColor0:LinearColor=&VC_DynamicScaleColor(
     Color to Scale:LinearColor=&MakeCustomLinearColorFromBool(Bool=$User.bIsPainted, True Color=$Particles.Initial.Color, False Color=$Particles.Initial.ColorTag0)
     RGB Scale:Vector3f=&VectorFromFloat(Value=&MakeCustomFloatFromBool(Bool=$User.bIsPainted, True Float=5, False Float=1))
     Alpha Scale:NiagaraFloat=1
    )
    Particles.MeshColor1:LinearColor=&VC_DynamicScaleColor(
     Color to Scale:LinearColor=&MakeCustomLinearColorFromBool(Bool=$User.bIsPainted, True Color=$Particles.Initial.Color, False Color=$Particles.Initial.ColorTag1)
     RGB Scale:Vector3f=&VectorFromFloat(Value=&MakeCustomFloatFromBool(Bool=$User.bIsPainted, True Float=5, False Float=1))
     Alpha Scale:NiagaraFloat=1
    )
    Particles.MeshColor2:LinearColor=&VC_DynamicScaleColor(
     Color to Scale:LinearColor=&MakeCustomLinearColorFromBool(Bool=$User.bIsPainted, True Color=$Particles.Initial.Color, False Color=$Particles.Initial.ColorTag2)
     RGB Scale:Vector3f=&VectorFromFloat(Value=&MakeCustomFloatFromBool(Bool=$User.bIsPainted, True Float=5, False Float=1))
     Alpha Scale:NiagaraFloat=1
    )
    Particles.MeshColor3:LinearColor=&VC_DynamicScaleColor(
     Color to Scale:LinearColor=&MakeCustomLinearColorFromBool(
      Bool=$User.bIsPainted
      True Color=$Particles.Initial.Color
      False Color=&Multiply_LinearColorByFloat(
       Linear Color=&ColorFromCurve(ColorCurve=DataInterface() {propertyValues {
          AlphaCurve {keys [<at 0 1 RCIM_Cubic>,<at 1 0 RCIM_Cubic>]}
          BlueCurve {keys [<at 0.1 0.375>,<at 0.3 2>,<at 0.5 1>]}
          GreenCurve {keys [<at 0.1 2>,<at 0.3 0.25>,<at 0.5 0.125>]}
          RedCurve {keys [<at 0.1 2>,<at 0.3 2>,<at 0.5 0.25>]}
         }})
       Float=1.8
      )
     )
     RGB Scale:Vector3f=&MakeCustomVectorFromBool(
      Bool=$User.bIsPainted
      True Vector=&VectorFromFloat(Value=&FloatFromCurve(
        FloatCurve=DataInterface() {propertyValues {Curve {defaultValue=1, keys [{tangentMode=RCTM_User, time=0, value=1}, {tangentMode=RCTM_User, time=1, value=10}]}}}
        Scale Curve=4
       ))
      False Vector=&VectorFromFloat(Value=&FloatFromCurve(FloatCurve=DataInterface() {propertyValues {Curve {keys [<at 0 1>,<at 1 5>]}}}))
     )
     Alpha Scale:NiagaraFloat=&FloatFromCurve(FloatCurve=DataInterface() {propertyValues {Curve {defaultValue=0, keys [{tangentMode=RCTM_User, time=0.8, value=1}, {tangentMode=RCTM_User, time=1, value=0}]}}})
    )
    Particles.MeshSize3:Vector3f=&Multiply_VectorByFloat(
     Vector=&VectorFromCurve(
      VectorCurve=DataInterface() {propertyValues {
        XCurve {keys [<at 0 1>,<at 1 1>]}
        YCurve {keys [<at 0 1 RCIM_Cubic>,<at 1 2 RCIM_Cubic>]}
        ZCurve {keys [<at 0 1 RCIM_Cubic>,<at 1 1.15 RCIM_Cubic>]}
       }}
      Scale Curve=<1 2 1>
     )
     Float=0.5
    )
    Particles.SpriteColor4:LinearColor=&VC_DynamicScaleColor(
     Color to Scale:LinearColor=$Particles.Initial.R1 Color
     RGB Scale:Vector3f=&VectorFromFloat(Value=&MakeCustomFloatFromBool(Bool=$User.bIsPainted, True Float=15, False Float=3))
     Alpha Scale:NiagaraFloat=1
    )
   ) {isSetParameters=true}
  ]
  rendererClasses [&NiagaraMeshRendererProperties, &NiagaraSpriteRendererProperties]
  renderers [
   {rendererClass=&NiagaraMeshRendererProperties, rendererIndex=0}
   {rendererClass=&NiagaraMeshRendererProperties, rendererIndex=1}
   {rendererClass=&NiagaraMeshRendererProperties, rendererIndex=2}
   {rendererClass=&NiagaraMeshRendererProperties, rendererIndex=3}
   {rendererClass=&NiagaraSpriteRendererProperties, rendererIndex=4}
  ]
 }
}
summary {
 systemName=NS_Boost_Magical
 emitters [
  {
   emitterName=NE_VC_Boost_Sprite_Parent002
   rendererClasses [&NiagaraMeshRendererProperties, &NiagaraSpriteRendererProperties]
  }
  {emitterName=NE_VC_Boost_Ribbon_Parent002, rendererClasses [&NiagaraRibbonRendererProperties]}
 ]
 userVariables [
  User.bIsFrontendPreview {type=NiagaraBool}
  User.bIsPainted {type=NiagaraBool}
  User.ColorVariantID {type=NiagaraInt32}
  User.DTCarOffset {type=Vector3f, description="", defaultValue {struct=Vector3f, value=<0 0 0>}}
  User.PaintedColor {type=LinearColor, description="", defaultValue {struct=LinearColor, value=<rgba 1 1 1 1>}}
  User.PaintedScalar {type=NiagaraFloat, defaultValue=1, description=""}
  User.PortCount {type=NiagaraInt32, defaultValue=1}
 ]
}
